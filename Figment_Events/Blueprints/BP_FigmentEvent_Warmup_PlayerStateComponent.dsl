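; Player state component for the Figment event warmup phase, retained as a rollback for logic now in BP_FigmentEvent_PawnComponent.
; On enable, it applies gameplay effects, grants items and an aura ability, resets AI bot targeting and configures respawn through a mutator.
; On disable, it removes effects, items and aura via stored handles and item lists.
; ScalableFloat toggles control item handling, bus-landing grant callback, drop prevention, fall-damage immunity and bot respawn.
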(event Custom|OnWarmupEnabled
  ;[1 0 0 0.5] TODO(Figment) - Consider removing this after 39.40. This is now implemented by BP_FigmentEvent_PawnComponent. Mantained just in case we need a rollback.
  (Utilities|SetupAbilitySystemComponent
    (:Is%20Valid
      (CallFunction|AddGameplayEffects)
      (CallFunction|GrantWarmupItems)
      (CallFunction|GrantAuraAbility)
      (CallFunction|StopAiAgression)
      (CallFunction|SetupRespawnAllowed))))

(event Custom|OnWarmupDisabled
  ;[1 0 0 0.5] TODO(Figment) - Consider removing this after 39.40. This is now implemented by BP_FigmentEvent_PawnComponent. Mantained just in case we need a rollback.
  (Utilities|SetupAbilitySystemComponent
    (:Is%20Valid
      (CallFunction|RemoveGameplayEffects)
      (CallFunction|RemoveWarmupItems)
      (CallFunction|ClearAuraAbility))))

(fn AddGameplayEffects
  (for _array_element (Variables|Default|GetGameplayEffects)
    (bind _returnvalue (GameplayEffects|ApplyGameplayEffectToSelf :self (Variables|Default|GetAbilitySystemComponent) :GameplayEffectClass _array_element :Level 0.0))
    (Utilities|Array|Add (Variables|Default|GetGameplayEffectHandles) _returnvalue)))

(fn ClearAuraAbility
  (GameplayAbilities|ClearAbility (Variables|Default|GetAbilitySystemComponent) (Variables|Default|GetAuraGameplayAbilityHandle)))

(fn GrantAuraAbility
  (bind _returnvalue (GameplayAbilities|GiveAbility (Variables|Default|GetAbilitySystemComponent) (Variables|Default|GetAuraGameplayAbilityClass) 0 -1))
  (Variables|Default|SetAuraGameplayAbilityHandle _returnvalue))

(fn GrantWarmupItems
  (if (ScalableFloat|GetValueasBool (Variables|Default|GetShouldGrantItems))
    (if (ScalableFloat|GetValueasBool (Variables|Default|GetShouldGrantItemsOnlyAfterLandingFromBus))
      (bind _asfort_player_state_zone (Utilities|Casting|CastToFortPlayerStateZone (Components|GetOwner)))
      (if (not (Class|FortPlayerStateZone|HasEverSkydivedfromBusandLanded _asfort_player_state_zone))
        (EventDispatchers|BindEventtoOnSkydivedfromBusandLanded _asfort_player_state_zone (EventDispatchers|CreateEvent :!event OnFirstLandedFromBus))))
    (for _array_element (Variables|Default|GetItemsToGrant)
      (bind _asplayer_state (Utilities|Casting|CastToPlayerState (Components|GetOwner)))
      (bind _asfort_player_controller (Utilities|Casting|CastToFortPlayerController (PlayerState|GetPlayerController _asplayer_state)))
      (if (== (Inventory|GetItemQuantityonInventoryOwner _asfort_player_controller _array_element) 0)
        (bind _returnvalue (Inventory|GiveItemtoInventoryOwner :InventoryOwner _asfort_player_controller :ItemDefinition _array_element :NumberToGive 1 :bNotifyPlayer false :ItemLevel -1 :PickupInstigatorHandle -1 :bUseItemPickupAnalyticEvent false :WeaponAmmoOverride -1))
        (if (ScalableFloat|GetValueasBool (Variables|Default|GetShouldPreventItemDropOnGrantedItems))
          (Inventory|SetItemPreventDropping _asfort_player_controller (Item|GetItemGuid _returnvalue) true))
        (else
          (Development|PrintString "Figment Warmup - Player already has warmup item. Not granting." true true "(R=1.000000,G=0.000000,B=0.603677,A=1.000000)"))))))

(fn OnFirstLandedFromBus
  (params (bSkydivedFromBusAndLanded :type Boolean))
  (if bSkydivedFromBusAndLanded
    (if (ScalableFloat|GetValueasBool (Variables|Default|GetShouldGrantItemsOnlyAfterLandingFromBus))
      (bind _asfort_player_state_zone (Utilities|Casting|CastToFortPlayerStateZone (Components|GetOwner)))
      (if (Class|FortPlayerStateZone|HasEverSkydivedfromBusandLanded _asfort_player_state_zone)
        (EventDispatchers|UnbindEventfromOnSkydivedfromBusandLanded _asfort_player_state_zone (EventDispatchers|CreateEvent :!event OnFirstLandedFromBus))
        (CallFunction|GrantWarmupItems)))))

(fn RemoveGameplayEffects
  (for _array_element (Variables|Default|GetGameplayEffectHandles)
    (GameplayEffects|RemoveActiveGameplayEffect (Variables|Default|GetAbilitySystemComponent) _array_element -1))
  (Utilities|Array|Clear (Variables|Default|GetGameplayEffectHandles)))

(fn RemoveWarmupItems
  (if (ScalableFloat|GetValueasBool (Variables|Default|GetShouldGrantItems))
    (if (ScalableFloat|GetValueasBool (Variables|Default|GetShouldRemoveGrantedItems))
      (for _array_element (Variables|Default|GetItemsToGrant)
        (bind _asplayer_state (Utilities|Casting|CastToPlayerState (Components|GetOwner)))
        (bind _asfort_player_controller (Utilities|Casting|CastToFortPlayerController (PlayerState|GetPlayerController _asplayer_state)))
        (Inventory|RemoveItemfromInventoryOwner _asfort_player_controller _array_element)
        (if (ScalableFloat|GetValueasBool (Variables|Default|GetShouldIgnoreNextFallOnRemoveItems))
          (bind _asfort_player_pawn (Utilities|Casting|CastToFortPlayerPawn (PlayerState|GetPawn _asplayer_state)))
          (Variables|/Script/FortniteGame.FortPawn|SetbIgnoreNextFallingDamage true _asfort_player_pawn))))))

(fn SetupRespawnAllowed
  (bind _asplayer_state (Utilities|Casting|CastToPlayerState (Components|GetOwner)))
  (LocalVariables|SetlocalIsBot (Variables|/Script/Engine.PlayerState|GetbIsABot _asplayer_state))
  (bind _asfort_player_controller_zone (Utilities|Casting|CastToFortPlayerControllerZone (PlayerState|GetPlayerController _asplayer_state)))
  (LocalVariables|SetlocalOwningPlayerController _asfort_player_controller_zone)
  (bind _returnvalue (Mutator|GetGameplayMutatorbyClass "/Figment_Events/Blueprints/Mutators/BP_FigmentEvent_RespawnMutator_PlayersOnly.BP_FigmentEvent_RespawnMutator_PlayersOnly_C"))
  (Class|FortAthenaMutatorSERespawn|SetPlayerRespawnAllowed _returnvalue (LocalVariables|GetlocalOwningPlayerController) (select (LocalVariables|GetlocalIsBot) (ScalableFloat|GetValueasBool (Variables|Default|GetShouldRespawnBot) 1.0) true) (Components|GetOwner)))

(fn StopAiAgression
  (bind _returnvalue (Components|GetOwner))
  (bind _returnvalue_2 (Fortnite|GetControllerfromActor _returnvalue))
  (Utilities|IsValid _returnvalue_2
    (:Is%20Valid
      (bind _asfort_athena_aibot_controller (Utilities|Casting|CastToFortAthenaAIBotController _returnvalue_2))
      (Class|FortAthenaAIBotController|ResetPerceptionandTargets _asfort_athena_aibot_controller "UpdateNow"))))
